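; Sprocket Poppy prop Blueprint, apparently derived from an Asteria tree parent, with an eyes component and a point light.
; Construction scales the light's attenuation radius by the largest actor scale axis.
; It also seeds a random stream from summed world location, writing the stable per-placement value to custom primitive data on the mesh and eyes, and prints it.
; BeginPlay reapplies that value; overlap and tick events are empty.
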(event EventBeginPlay
  (bind _randomvalue (Variables|Default|GetRandomValue))
  (|CallParentFunction)
  (Rendering|Material|SetCustomPrimitiveDataFloat (Variables|Debug|GetStaticMeshComponent) 30 _randomvalue)
  (Rendering|Material|SetCustomPrimitiveDataFloat (Variables|Default|GetSprocketPoppyEyes) 30 _randomvalue)
  (Utilities|FlowControl|Sequence))

(event Collision|EventActorBeginOverlap
  (params (OtherActor :type Actor%20Object%20Reference)))

(event EventTick
  (params (DeltaSeconds :type Float%20%28single-precision%29)))

(fn ConstructionScript
  (bind _stream (Variables|Default|GetStream))
  (|CallParentFunction)
  (Utilities|FlowControl|Sequence
    (:then_0
      (Rendering|Lighting|SetAttenuationRadius (Variables|Default|GetPointLight) (* (Variables|Default|GetLightRadius) (Math|Float|Max(Float) (.x (Transformation|GetActorScale3D)) (.y (Transformation|GetActorScale3D)) (.z (Transformation|GetActorScale3D))))))
    (:then_1
      (Math|Random|SetRandomStreamSeed _stream (Math|Float|Truncate (Utilities|Operators|Add (.x (Transformation|GetActorLocation)) (.y (Transformation|GetActorLocation)) (.z (Transformation|GetActorLocation)))))
      (bind _output_get (Variables|Default|SetRandomValue (Math|Random|RandomFloatinRangefromStream _stream 0.1 0.9)))
      (Rendering|Material|SetDefaultCustomPrimitiveDataFloat (Variables|Debug|GetStaticMeshComponent) 30 _output_get)
      (Rendering|Material|SetDefaultCustomPrimitiveDataFloat (Variables|Default|GetSprocketPoppyEyes) 30 _output_get))
    (:then_2
      (Development|PrintString (Utilities|String|Append "randomValue = " (Utilities|String|ToString(Float) (Variables|Default|GetRandomValue)))))))
