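# Masked, two-sided lit material for Sprocket Poppy leaf Niagara mesh particles, using the prop's leaf base color, normal and mask textures.
# The inverted blue mask channel times particle color feeds an emissive adjustment function.
# Panning Gaussian noise, scaled by the WorldWPO dynamic parameter, multiplies a scaled object position, apparently for world position offset motion.
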Material /Hermes/Props/SprocketPoppy/FX/M_SprocketPoppy_Leaf
graph {
 properties {
  bAutomaticallySetUsageInEditor=false
  bUsedWithNiagaraMeshParticles=true
  blendMode=BLEND_Masked
  twoSided=true
  usedShadingModels=Default Lit
 }
 nodes [
  n1 = TextureSample() {samplerType=SAMPLERTYPE_Masks, texture=/Hermes/Props/SprocketPoppy/Textures/T_Hermes_SprocketPoppy_Leaves_A_F_OSSR}
  MaterialFunctionCall(Emissive=Multiply(A=OneMinus(None=@n1.B), B=ParticleColor().RGB)) {materialFunction=&MF_FX_EmissiveAdjustments}
  n2 = TextureSample(UVs=Panner(Coordinate=TextureCoordinate()) {speedX=0.3, speedY=0.2}) {texture=/Game/Effects/Fort_Effects/Textures/Tile/T_GaussianNoise_01_M}
  Multiply(
   A=Divide(A=ObjectPositionWS()) {constB=15}
   B=Multiply(A=Multiply(A=@n2.RGB) {constB=0.5}, B=DynamicParameter() {defaultValue=<rgba 0 1 1 1>, paramNames [WorldWPO,Param2,Param3,Param4]}.WorldWPO) {constB=0}
  )
  TextureSample() {texture=/Hermes/Props/SprocketPoppy/Textures/T_Hermes_SprocketPoppy_Leaves_A_F_basecolor}
  TextureSample() {samplerType=SAMPLERTYPE_Normal, texture=/Hermes/Props/SprocketPoppy/Textures/T_Hermes_SprocketPoppy_Leaves_A_F_normal}
 ]
}
