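# Material function that procedurally animates the Sprocket Poppy prop's face, modifying incoming material attributes behind a static switch.
# Pivot-based, mirrored local-space masks drive world position offset for mouth, cheeks, brows, eyes and a head look-at toward PlayerPosition, fading with distance and angle.
# Curve-atlas timelines, sine variation and per-primitive random data vary the motion, alongside adjusted normals and randomly enabled eye glow.
# Extra outputs expose look-at and masks; a switch adds debug visualization.
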MaterialFunction /Hermes/Props/SprocketPoppy/Materials/FaceAnim/MF_SprocketPoppy_Anim
graph {nodes [
  n1 = ScalarParameter() {defaultValue=-20.1459, group=SprocketPoppy, parameterName=EyeBottomMask}
  n2 = NamedRerouteUsage()
  n3 = ComponentMask(None=@n2) {b=true}
  n4 = Saturate(None=Sign(None=@n3))
  n5 = ScalarParameter() {defaultValue=31.1856, group=SprocketPoppy, parameterName=EyeUpDownMask}
  Add(A=Multiply(A=LinearInterpolate(A=@n1, Alpha=@n4), B=ComponentMask(None=@n2) {b=true}), B=Divide(A=@n5, B=@n1))
  n6 = NamedRerouteUsage()
  n7 = Multiply(
   A=Power(Base=@n6, Exp=ScalarParameter() {defaultValue=4.0407, group=SprocketPoppy, parameterName=EyeGlow_Pow})
   B=VectorParameter() {group=SprocketPoppy, parameterName=Eye_Glow, defaultValue=<rgba 0.2582 0.302 0.0556 1>}.RGB
  )
  n8 = VectorParameter() {group=SprocketPoppy, parameterName=Eye_Glow2, defaultValue=<rgba 1 0.6968 0.0845 1>}
  n9 = Constant3Vector() {constant=<rgba 0 0 1 0>}
  n10 = Constant3Vector() {constant=<rgba 0 1 0 0>}
  n11 = Constant() {r=1}
  n12 = Multiply(
   A=NamedRerouteUsage()
   B=MaterialFunctionCall(X=NamedRerouteUsage(), Y=@n11, Z=@n11) {materialFunction=&MakeFloat3}.Result
  )
  n13 = SphereMask(
   A=MaterialFunctionCall(
    Plane Normal Axis=Transform(None=@n9) {transformSourceType=TRANSFORMSOURCE_Camera}
    Plane X Axis=Transform(None=@n10) {transformSourceType=TRANSFORMSOURCE_Camera}
    Plane Center=TransformPosition(None=@n12) {transformType=TRANSFORMPOSSOURCE_World}
    UV Scale=ScalarParameter() {defaultValue=17.8587, group=SprocketPoppy, parameterName=EyeGlowSize}
   ) {materialFunction=&VirtualPlaneCoordinates}.UVs
   B=Constant2Vector() {g=0.5, r=0.5}
   Hardness=ScalarParameter() {group=SprocketPoppy, parameterName=EyeGlowHardness}
  ) {attenuationRadius=0.5}
  Add(A=@n7, B=Multiply(A=OneMinus(None=@n6), B=Multiply(A=@n8.RGB, B=@n13)))
  CameraPositionWS()
  Constant3Vector()
  Constant() {r=1}
  Constant() {r=1}
  Constant() {r=-1}
  n14 = Constant3Vector()
  n15 = MaterialFunctionCall(Float3=Abs(None=TransformPosition(None=@n14) {transformType=TRANSFORMPOSSOURCE_World})) {materialFunction=&BreakOutFloat3Components}
  Divide(A=Add(A=Add(A=@n15.R, B=@n15.R), B=@n15.B)) {constB=10}
  n16 = ScalarParameter() {defaultValue=0.04, group=SprocketPoppy, parameterName=SpeedMouthAnim}
  n17 = Multiply(A=Add(A=Time(), B=NamedRerouteUsage()), B=@n16)
  Divide(A=Add(A=Sine(None=@n17)))
  n18 = NamedRerouteDeclaration(None=FunctionInput() {inputType=FunctionInput_MaterialAttributes}) {name=InMaterial, nodeColor=<rgba 0.0547 1 0 1>}
  n19 = NamedRerouteUsage()
  n20 = NamedRerouteUsage()
  n21 = Add(A=Add(A=Add(A=@n19, B=@n20), B=NamedRerouteUsage()), B=NamedRerouteUsage())
  n22 = NamedRerouteUsage()
  n23 = SetMaterialAttributes!(
   MaterialAttributes=@n18
   WorldPositionOffset=Transform(None=Add(A=@n21, B=@n22)) {transformSourceType=TRANSFORMSOURCE_Local}
   Normal=NamedRerouteUsage()
   EmissiveColor=NamedRerouteUsage()
  ) {attributeSetTypes [#1, #2, #3]}
  n24 = GetMaterialAttributes(None=@n23) {attributeGetTypes [#1]}
  n25 = NamedRerouteUsage()
  n26 = ScalarParameter() {defaultValue=2, group=SprocketPoppy, parameterName=SprocketPoppyDebugDisplay}
  n27 = LinearInterpolate(
   A=MaterialFunctionCall(R=@n25, G=NamedRerouteUsage(), B=NamedRerouteUsage()) {materialFunction=&AppendMany}.RGB
   B=MaterialFunctionCall(R=NamedRerouteUsage(), G=NamedRerouteUsage(), B=NamedRerouteUsage()) {materialFunction=&AppendMany}.RGB
   Alpha=@n26
  )
  n28 = MaterialFunctionCall(R=@n25, G=NamedRerouteUsage(), B=NamedRerouteUsage()) {materialFunction=&AppendMany}
  n29 = Saturate(None=Subtract(A=@n26))
  n30 = MaterialFunctionCall(
   Plane Normal Axis=Constant3Vector() {constant=<rgba 0 1 0 0>}
   Plane X Axis=Constant3Vector() {constant=<rgba 1 0 0 0>}
   Plane Center=TransformPosition(None=NamedRerouteUsage()) {transformType=TRANSFORMPOSSOURCE_World}
   UV Scale=ScalarParameter() {defaultValue=33.6802, group=SprocketPoppy, parameterName=DebugPlaneSize}
  ) {materialFunction=&VirtualPlaneCoordinates}
  n31 = Constant2Vector() {g=0.5, r=0.5}
  n32 = MaterialFunctionCall(
   Plane Normal Axis=Constant3Vector() {constant=<rgba 0 1 0 0>}
   Plane X Axis=Constant3Vector() {constant=<rgba 1 0 0 0>}
   Plane Center=TransformPosition(None=NamedRerouteUsage()) {transformType=TRANSFORMPOSSOURCE_World}
   UV Scale=ScalarParameter() {defaultValue=6.9994, group=SprocketPoppy, parameterName=DebugPlaneSize}
  ) {materialFunction=&VirtualPlaneCoordinates}
  n33 = Constant2Vector() {g=0.5, r=0.5}
  n34 = Add(
   A=Multiply(A=SphereMask(A=@n30.UVs, B=@n31) {attenuationRadius=0.5}, B=Constant3Vector() {constant=<rgba 5 0 0 0>})
   B=Multiply(A=SphereMask(A=@n32.UVs, B=@n33) {attenuationRadius=0.5}, B=Constant3Vector() {constant=<rgba 20 0 0 0>})
  )
  n35 = MaterialFunctionCall(
   Plane Normal Axis=Constant3Vector() {constant=<rgba 0 1 0 0>}
   Plane X Axis=Constant3Vector() {constant=<rgba 1 0 0 0>}
   Plane Center=TransformPosition(None=NamedRerouteUsage()) {transformType=TRANSFORMPOSSOURCE_World}
   UV Scale=ScalarParameter() {defaultValue=6.9994, group=SprocketPoppy, parameterName=DebugPlaneSize}
  ) {materialFunction=&VirtualPlaneCoordinates}
  n36 = Constant2Vector() {g=0.5, r=0.5}
  n37 = MaterialFunctionCall(
   Plane Normal Axis=Constant3Vector() {constant=<rgba 0 1 0 0>}
   Plane X Axis=Constant3Vector() {constant=<rgba 1 0 0 0>}
   Plane Center=TransformPosition(None=NamedRerouteUsage()) {transformType=TRANSFORMPOSSOURCE_World}
   UV Scale=ScalarParameter() {defaultValue=6.9994, group=SprocketPoppy, parameterName=DebugPlaneSize}
  ) {materialFunction=&VirtualPlaneCoordinates}
  n38 = Constant2Vector() {g=0.5, r=0.5}
  n39 = Add(
   A=Multiply(A=SphereMask(A=@n35.UVs, B=@n36) {attenuationRadius=0.5}, B=Constant3Vector() {constant=<rgba 1.1819 20 0 1>})
   B=Multiply(A=SphereMask(A=@n37.UVs, B=@n38) {attenuationRadius=0.5}, B=Constant3Vector() {constant=<rgba 0 1.082 20 1>})
  )
  n40 = ScalarParameter() {group=SprocketPoppy, parameterName=SprocketPoppyDebugDisplayWPO}
  n41 = StaticSwitchParameter(
   True=SetMaterialAttributes!(
    MaterialAttributes=@n24.MaterialAttributes
    BaseColor=LinearInterpolate(A=@n27, B=@n28.RGB, Alpha=@n29)
    EmissiveColor=Add(A=@n34, B=@n39)
    WorldPositionOffset=LinearInterpolate(B=@n24.WorldPositionOffset, Alpha=@n40)
   ) {attributeSetTypes [#4, #3, #1]}
   False=@n23
  ) {group=SprocketPoppy, parameterName=SprocketPoppyDebug}
  FunctionOutput(None=StaticSwitchParameter(True=@n41, False=@n18) {group=SprocketPoppy, parameterName=SprocketPoppy_Anim})
  n42 = LocalPosition() {includedOffsets=ExcludeOffsets}
  n43 = VectorParameter() {group=SprocketPoppy, parameterName=WholeFacePivot, defaultValue=<rgba 0 0 275.6153 1>}
  n44 = Subtract(A=@n42.XYZ, B=@n43.RGB)
  n45 = Multiply(A=NamedRerouteUsage(), B=ScalarParameter() {defaultValue=0.5, group=SprocketPoppy, parameterName=MaxAmbientHeadTurn})
  n46 = CollectionParameter() {collection=&FortniteMaterialParameters, parameterName=PlayerPosition}
  n47 = Subtract(A=@n43.RGB, B=TransformPosition(None=@n46) {transformSourceType=TRANSFORMPOSSOURCE_World})
  n48 = Normalize(VectorInput=@n47)
  n49 = Constant3Vector() {constant=<rgba 0 1 0 0>}
  n50 = Multiply(A=@n48, B=Sign(None=DotProduct(A=@n49, B=@n48)))
  n51 = TransformPosition(None=@n43.RGB) {transformType=TRANSFORMPOSSOURCE_TranslatedWorld}
  n52 = Length(None=@n51) {is=class MaterialExpressionLength}
  n53 = Multiply(
   A=ScalarParameter() {defaultValue=1737.1581, group=SprocketPoppy, parameterName=LookAtTransitionDistance}
   B=MaterialFunctionCall() {materialFunction=&ObjectScale}.Scale Z
  )
  n54 = Saturate(None=OneMinus(None=Divide(A=@n52, B=@n53)))
  n55 = Multiply(A=@n47, B=Constant3Vector() {constant=<rgba 1 1 0 0>})
  n56 = Abs(None=DotProduct(A=Normalize(VectorInput=@n55), B=Constant3Vector() {constant=<rgba 1 0 0 0>}))
  n57 = ScalarParameter() {defaultValue=15, group=SprocketPoppy, parameterName=LookAtFade}
  n58 = OneMinus(None=Saturate(None=Power(Base=@n56, Exp=@n57)))
  n59 = Multiply(A=@n54, B=@n58)
  n60 = VectorParameter() {group=SprocketPoppy, parameterName=LookAtDirectionVar, defaultValue=<rgba -0.5 1 -0.1 1>}
  n61 = VectorParameter() {group=SprocketPoppy, parameterName=LookAtDirection, defaultValue=<rgba 0.5 1 0.1 1>}
  n62 = Normalize(VectorInput=Clamp(None=LinearInterpolate(A=@n45, B=@n50, Alpha=@n59), Min=@n60.RGB, Max=@n61.RGB))
  n63 = Normalize(VectorInput=CrossProduct(A=@n62, B=Constant3Vector() {constant=<rgba 0 0 1 0>}))
  n64 = CrossProduct(A=@n63, B=@n62)
  n65 = Subtract(
   A=MaterialFunctionCall(VectorToTransform=@n44, BasisX=@n63, BasisY=@n62, BasisZ=Normalize(VectorInput=@n64)) {materialFunction=&Transform3x3Matrix}.Transformed Vector
   B=@n44
  )
  n66 = Subtract(A=@n42.XYZ, B=@n43.RGB)
  n67 = Divide(
   A=Subtract(A=Length(None=@n66), B=ScalarParameter() {defaultValue=95, group=SprocketPoppy, parameterName=WholeaceMaskRadiusBias})
   B=ScalarParameter() {defaultValue=106.1737, group=SprocketPoppy, parameterName=WholeaceMaskRadius}
  )
  n68 = NamedRerouteDeclaration(None=Saturate(None=OneMinus(None=@n67))) {name=WholeFaceMask, nodeColor=<rgba 0 1 0.4141 1>}
  FunctionOutput(None=NamedRerouteDeclaration(None=Multiply(A=@n65, B=@n68)) {name=LookAtTransformLocal, nodeColor=<rgba 0.2656 1 0 1>}) {outputName=LookatTransformLocalOnly, sortPriority=33}
  n69 = MaterialFunctionCall(Float3=LocalPosition() {includedOffsets=ExcludeOffsets}.XYZ) {materialFunction=&BreakOutFloat3Components}
  n70 = Abs(None=Add(A=ScalarParameter() {defaultValue=-10.7222, group=SprocketPoppy, parameterName=FaceCenterBias}, B=@n69.G))
  n71 = MaterialFunctionCall(X=@n69.R, Y=@n70, Z=@n69.R) {materialFunction=&MakeFloat3}
  n72 = VertexInterpolator(VS=@n71.Result)
  FunctionOutput(None=NamedRerouteDeclaration(None=ShaderStageSwitch(PixelShader=@n72.PS, VertexShader=@n71.Result)) {name=CustomMirroredLocalSpace, nodeColor=<rgba 0 1 0.9531 1>}) {outputName=CustomMirroredLocalSpace, sortPriority=99}
  FunctionOutput(None=@n54) {outputName=LookAtTransition mask, sortPriority=456}
  n73 = NamedRerouteUsage()
  n74 = ScalarParameter() {defaultValue=0.5, group=SprocketPoppy, parameterName=RandomEyeGlowThreshold}
  n75 = Saturate(None=Ceil(None=Subtract(A=@n73, B=@n74)))
  FunctionOutput(None=@n75) {outputName=RandomMask}
  LinearInterpolate(A=@n60.RGB, B=@n61.RGB, Alpha=NamedRerouteUsage())
  LocalPosition() {includedOffsets=ExcludeOffsets}
  n76 = ScalarParameter() {defaultValue=-0.4, group=SprocketPoppy, parameterName=RotationEyes}
  Multiply(A=NamedRerouteUsage(), B=@n76)
  Multiply(A=@n43.RGB, B=NamedRerouteUsage())
  Multiply(A=@n43.RGB, B=Constant3Vector() {constant=<rgba 0 0 1 0>})
  Multiply(A=@n21, B=NamedRerouteUsage())
  n77 = CurveAtlasRowParameter(CurveTime=Frac(None=@n17)) {
   atlas=&/Curve_SprocketPoppy_FaceAnim
   curve=&/C_SprocketPoppy_Mouth2
   defaultValue=2
   parameterName=CurveMouth
  }
  NamedRerouteDeclaration(None=StaticSwitchParameter(True=Constant() {r=1}, False=@n77.R) {group=SprocketPoppy, parameterName=SprocketPoppyDebugAnimMaxPose}) {name=TimeLine, nodeColor=<rgba 0.9688 1 0 1>}
  n78 = NamedRerouteUsage()
  n79 = NamedRerouteDeclaration(None=VectorParameter() {group=SprocketPoppy, parameterName=NosePivot, defaultValue=<rgba 0 152.4677 270.0145 1>}.RGB) {name=NosePivot, nodeColor=<rgba 0.1719 1 0 1>}
  n80 = Divide(A=Length(None=Subtract(A=@n78, B=@n79)), B=ScalarParameter() {defaultValue=68.0323, group=SprocketPoppy, parameterName=NoseMaskRadius})
  NamedRerouteDeclaration(None=Saturate(None=OneMinus(None=@n80))) {name=NoseMask, nodeColor=<rgba 1 0 0.8906 1>}
  n81 = ScalarParameter() {bUseCustomPrimitiveData=true, parameterName=RandomValue, primitiveDataIndex=30}
  n82 = Multiply(A=@n81) {constB=33}
  NamedRerouteDeclaration(None=@n82) {name=Random, nodeColor=<rgba 0.9688 1 0 1>}
  NamedRerouteDeclaration(None=@n81) {name=Random_Frac, nodeColor=<rgba 0.9688 1 0 1>}
  n83 = Add(A=Time(), B=NamedRerouteUsage())
  n84 = Multiply(A=@n83, B=ScalarParameter() {defaultValue=0.4237, group=SprocketPoppy, parameterName=VariationSpeed1})
  n85 = Divide(A=Add(A=Sine(None=@n84)))
  n86 = Sine(None=Multiply(A=@n83, B=ScalarParameter() {defaultValue=0.3357, group=SprocketPoppy, parameterName=EyeLookVariationSpeed}))
  NamedRerouteDeclaration(None=Saturate(None=Add(A=@n85, B=@n86))) {name=Variation, nodeColor=<rgba 0.9453 0 1 1>}
  n87 = MaterialFunctionCall(Float3=NamedRerouteUsage()) {materialFunction=&BreakOutFloat3Components}
  n88 = MaterialFunctionCall(X=Abs(None=@n87.R), Y=@n87.G, Z=@n87.G) {materialFunction=&MakeFloat3}
  n89 = NamedRerouteDeclaration(None=VectorParameter() {group=SprocketPoppy, parameterName=CheekPivot, defaultValue=<rgba 34.8411 59.7118 264.7343 1>}.RGB) {name=CheekPivot, nodeColor=<rgba 0 0.9062 1 1>}
  n90 = Divide(
   A=Length(None=Subtract(A=@n88.Result, B=@n89))
   B=ScalarParameter() {defaultValue=83.3086, group=SprocketPoppy, parameterName=CheekMaskRadius}
  )
  n91 = NamedRerouteDeclaration(None=Saturate(None=OneMinus(None=@n90))) {name=CheekMask, nodeColor=<rgba 1 0.0938 0 1>}
  n92 = LinearInterpolate(
   B=LinearInterpolate(
    A=ScalarParameter() {defaultValue=-5, group=SprocketPoppy, parameterName=CheeksUPvar}
    B=ScalarParameter() {defaultValue=25, group=SprocketPoppy, parameterName=CheeksUP}
    Alpha=NamedRerouteUsage()
   )
   Alpha=NamedRerouteUsage()
  )
  NamedRerouteDeclaration(None=Multiply(A=Constant3Vector() {constant=<rgba 0 0 1 0>}, B=Multiply(A=@n91, B=@n92))) {name=CheekTransform Local Space, nodeColor=<rgba 0 0.0391 1 1>}
  n93 = Add(A=@n1, B=@n3)
  n94 = OneMinus(None=Divide(A=Abs(None=@n93), B=@n5))
  n95 = NamedRerouteDeclaration(None=Multiply(A=Saturate(None=@n94), B=NamedRerouteUsage())) {name=EyeBrowMask, nodeColor=<rgba 0 0.8828 1 1>}
  n96 = LinearInterpolate(
   B=LinearInterpolate(
    A=ScalarParameter() {defaultValue=-14.7358, group=SprocketPoppy, parameterName=EyeBrowRaiseVar}
    B=ScalarParameter() {defaultValue=-35, group=SprocketPoppy, parameterName=EyeBrowRaise}
    Alpha=NamedRerouteUsage()
   )
   Alpha=NamedRerouteUsage()
  )
  NamedRerouteDeclaration(None=Multiply(A=Constant3Vector() {constant=<rgba 0 0 1 0>}, B=Multiply(A=@n95, B=@n96))) {name=EyeBrowTransform Local, nodeColor=<rgba 1 0 0.7734 1>}
  n97 = NamedRerouteUsage()
  n98 = MaterialFunctionCall(X=@n97, Y=@n97, Z=@n97) {materialFunction=&MakeFloat3}
  n99 = MaterialFunctionCall(Float3=NamedRerouteUsage()) {materialFunction=&BreakOutFloat3Components}
  n100 = Sign(None=@n99.R)
  n101 = MaterialFunctionCall(X=@n100, Y=NamedRerouteUsage(), Z=Constant() {r=1}) {materialFunction=&MakeFloat3}
  n102 = Multiply(A=VectorParameter() {group=SprocketPoppy, parameterName=RotationAxisEye, defaultValue=<rgba 1 0.5823 0 1>}.RGB, B=@n101.Result)
  n103 = VectorParameter() {group=SprocketPoppy, parameterName=RotationAxisEyeVariation, defaultValue=<rgba 0 0 0.304 1>}
  n104 = NamedRerouteUsage()
  n105 = Normalize(VectorInput=Add(A=Normalize(VectorInput=@n102), B=LinearInterpolate(B=@n103.RGB, Alpha=@n104)))
  n106 = MaterialFunctionCall(X=Abs(None=@n99.R), Y=@n99.G, Z=@n99.G) {materialFunction=&MakeFloat3}
  n107 = NamedRerouteDeclaration(None=VectorParameter() {group=SprocketPoppy, parameterName=EyePivot, defaultValue=<rgba 27.921 58.7808 303.5045 1>}.RGB) {name=EyePivot, nodeColor=<rgba 1 0 0.3984 1>}
  n108 = Length(None=NamedRerouteDeclaration(None=Subtract(A=@n106.Result, B=@n107)) {name=EyeLocalSpacePosition, nodeColor=<rgba 0.4531 1 0 1>})
  n109 = ScalarParameter() {defaultValue=24.3133, group=SprocketPoppy, parameterName=EyeMaskRadiusInner}
  n110 = NamedRerouteDeclaration(None=Saturate(None=Divide(A=@n108, B=@n109) {constB=1})) {name=EyeMask_Inner, nodeColor=<rgba 0.5469 1 0 1>}
  n111 = ScalarParameter() {defaultValue=56.3395, group=SprocketPoppy, parameterName=EyeMaskRadius}
  n112 = Saturate(None=OneMinus(None=Divide(A=@n108, B=@n111)))
  n113 = Power(
   Base=NamedRerouteDeclaration(None=Multiply(A=@n110, B=@n112)) {name=EyeMask, nodeColor=<rgba 0.5469 1 0 1>}
   Exp=ScalarParameter() {defaultValue=2.0526, group=SprocketPoppy, parameterName=EyeMaskRadiusFalloff}
  )
  n114 = Multiply(A=@n100, B=LinearInterpolate(B=@n76, Alpha=NamedRerouteUsage()))
  n115 = Multiply(A=@n101.Result, B=@n107)
  NamedRerouteDeclaration(None=RotateAboutAxis(
    NormalizedRotationAxis=Multiply(A=@n98.Result, B=@n105)
    RotationAngle=Multiply(A=@n113, B=@n114)
    PivotPoint=NamedRerouteDeclaration(None=@n115) {name=EyePivot_Mirrored, nodeColor=<rgba 1 0 0.3984 1>}
    Position=LocalPosition() {includedOffsets=ExcludeOffsets}.XYZ
   )) {name=Eyes Rotation Transform Local, nodeColor=<rgba 0 1 0.9063 1>}
  n116 = ScalarParameter() {defaultValue=0.5, group=SprocketPoppy, parameterName=NoseMaskAddtoMouth}
  n117 = NamedRerouteUsage()
  n118 = Subtract(
   A=NamedRerouteUsage()
   B=NamedRerouteDeclaration(None=VectorParameter() {group=SprocketPoppy, parameterName=MouthPivot, defaultValue=<rgba 0 80.8556 224.5394 1>}.RGB) {name=MouthPivot, nodeColor=<rgba 0 0.3906 1 1>}
  )
  n119 = DotProduct(A=@n118, B=Constant3Vector() {constant=<rgba 0 0 1 1>})
  n120 = ScalarParameter() {defaultValue=40.4711, group=SprocketPoppy, parameterName=MouthMaskUpGradient}
  n121 = Saturate(None=OneMinus(None=Divide(A=@n119, B=@n120)))
  n122 = Length(None=@n118)
  n123 = ScalarParameter() {defaultValue=71.9427, group=SprocketPoppy, parameterName=MouthMaskRadiusInner}
  n124 = Saturate(None=Power(Base=Divide(A=@n122, B=@n123), Exp=ScalarParameter() {defaultValue=0.7908, group=SprocketPoppy, parameterName=MouthMaskRadiusInnerFalloff}))
  n125 = VectorParameter() {group=SprocketPoppy, parameterName=MouthMaskScale, defaultValue=<rgba 0.2846 0.297 0.5095 1>}
  n126 = ComponentMask(None=@n125.RGB) {g=true, r=true}
  n127 = ScalarParameter() {defaultValue=0.0234, parameterName=MouthMaskScaleBottom}
  n128 = Saturate(None=Sign(None=ComponentMask(None=@n118) {b=true}))
  n129 = Multiply(A=@n118, B=LinearInterpolate(A=AppendVector(A=@n126, B=@n127), B=@n125.RGB, Alpha=@n128))
  n130 = OneMinus(None=Divide(A=Length(None=@n129), B=ScalarParameter() {defaultValue=22.46, group=SprocketPoppy, parameterName=MouthMaskRadius}))
  n131 = Multiply(
   A=Multiply(
    A=NamedRerouteDeclaration(None=@n121) {name=MouthMask_UpGradient, nodeColor=<rgba 0 0.2266 1 1>}
    B=NamedRerouteDeclaration(None=@n124) {name=MouthMask_Inner, nodeColor=<rgba 0 0.2266 1 1>}
   )
   B=NamedRerouteDeclaration(None=Saturate(None=@n130)) {name=MouthMask, nodeColor=<rgba 0 0.2266 1 1>}
  )
  n132 = Multiply(A=OneMinus(None=@n128), B=OneMinus(None=@n124))
  n133 = Divide(
   A=Abs(None=ComponentMask(None=@n118) {b=true})
   B=ScalarParameter() {defaultValue=278.0421, group=SprocketPoppy, parameterName=MouthBottomFade}
  )
  n134 = Multiply(A=Add(A=Multiply(A=@n116, B=@n117), B=Add(A=@n131, B=@n132)), B=Saturate(None=OneMinus(None=@n133)))
  n135 = LinearInterpolate(
   B=LinearInterpolate(
    A=ScalarParameter() {defaultValue=57.6365, group=SprocketPoppy, parameterName=MouthOpenVar}
    B=ScalarParameter() {defaultValue=90, group=SprocketPoppy, parameterName=MouthOpen}
    Alpha=NamedRerouteUsage()
   )
   Alpha=NamedRerouteUsage()
  )
  NamedRerouteDeclaration(None=Multiply(A=Constant3Vector() {constant=<rgba 0 0 -1 0>}, B=Multiply(A=@n134, B=@n135))) {name=Mouth Transform Local, nodeColor=<rgba 1 0.5859 0 1>}
  n136 = Sign(None=@n69.G)
  NamedRerouteDeclaration(None=ShaderStageSwitch(PixelShader=VertexInterpolator(VS=@n136).PS, VertexShader=@n136)) {name=MirrorSign, nodeColor=<rgba 0 1 0.9531 1>}
  n137 = Constant() {r=1}
  n138 = MaterialFunctionCall(X=@n137, Y=@n136, Z=@n137) {materialFunction=&MakeFloat3}
  NamedRerouteDeclaration(None=ShaderStageSwitch(PixelShader=VertexInterpolator(VS=@n138.Result).PS, VertexShader=@n138.Result)) {name=MirrorSign_V3, nodeColor=<rgba 0 1 0.9531 1>}
  NamedRerouteDeclaration(None=@n85) {name=Variation_Normalized, nodeColor=<rgba 0.9453 0 1 1>}
  n139 = Multiply(A=Saturate(None=NamedRerouteUsage())) {constB=0.5}
  n140 = Divide(A=Constant() {r=1}, B=@n16)
  NamedRerouteDeclaration(None=Add(A=Multiply(A=@n139, B=@n140), B=@n82)) {name=Random_Frac_FaceAnim, nodeColor=<rgba 0.9688 1 0 1>}
  NamedRerouteDeclaration(None=@n58) {name=Debug color, nodeColor=<rgba 0.6875 1 0 1>}
  n141 = Saturate(None=Divide(A=@n70, B=ScalarParameter() {defaultValue=60283.4453, group=SprocketPoppy, parameterName=MirroCenterMask}))
  NamedRerouteDeclaration(None=ShaderStageSwitch(PixelShader=VertexInterpolator(VS=@n141).PS, VertexShader=@n141)) {name=MirrorCenterMask, nodeColor=<rgba 0 1 0.9531 1>}
  NamedRerouteDeclaration(None=Multiply(A=@n7, B=@n75)) {name=Emissive, nodeColor=<rgba 1 0 0.9141 1>}
  n142 = GetMaterialAttributes(None=NamedRerouteUsage()) {attributeGetTypes [#2]}
  n143 = LinearInterpolate(A=Constant(), B=ScalarParameter() {defaultValue=-2, group=SprocketPoppy, parameterName=FaceNormalIntesity}, Alpha=NamedRerouteUsage())
  n144 = Add(A=Add(A=NamedRerouteUsage(), B=NamedRerouteUsage()), B=NamedRerouteUsage())
  NamedRerouteDeclaration(None=LinearInterpolate(
    A=@n142.Normal
    B=MaterialFunctionCall(Normal=@n142.Normal, Flatness=@n143) {materialFunction=&FlattenNormal}.Result
    Alpha=Saturate(None=@n144)
   )) {name=ModifiedNormal, nodeColor=<rgba 0.6406 1 0 1>}
  n145 = WorldPosition()
  n146 = Multiply(
   A=Divide(A=ComponentMask(None=@n145.XYZ) {b=true, r=true}, B=ScalarParameter() {defaultValue=64, parameterName=DebugTiling})
   B=Constant2Vector() {g=-1, r=-1}
  )
  NamedRerouteDeclaration(None=MaterialFunctionCall(Number=@n81, UVs=Frac(None=@n146)) {materialFunction=&DebugScalarValues}.Result) {name=Debug Value, nodeColor=<rgba 1 0 0.4687 1>}
  n147 = Sine(None=Multiply(A=@n83, B=ScalarParameter() {defaultValue=0.5837, group=SprocketPoppy, parameterName=VariationSpeed2}))
  NamedRerouteDeclaration(None=Divide(A=Add(A=@n147))) {name=Variation_Normalized_2, nodeColor=<rgba 0.9453 0 1 1>}
  n148 = Sine(None=Multiply(A=@n83, B=ScalarParameter() {defaultValue=0.2637, group=SprocketPoppy, parameterName=VariationSpeed3}))
  NamedRerouteDeclaration(None=Divide(A=Add(A=@n148))) {name=Variation_Normalized_3, nodeColor=<rgba 0.9453 0 1 1>}
  n149 = Time()
  n150 = ScalarParameter() {defaultValue=0.04, group=SprocketPoppy, parameterName=SpeedHeadAnim}
  n151 = Add(A=@n150, B=Multiply(A=@n150, B=NamedRerouteUsage()))
  n152 = CurveAtlasRowParameter(CurveTime=Frac(None=Multiply(A=@n149, B=@n151))) {
   atlas=&Curve_SprocketPoppy_FaceAnim
   curve=&/C_SprocketPoppy_Head
   defaultValue=1
   parameterName=CurveHead
  }
  NamedRerouteDeclaration(None=Subtract(A=@n152) {constB=0.5}) {name=HeadAnim, nodeColor=<rgba 0 1 0.4844 1>}
  NamedRerouteUsage()
  NamedRerouteUsage()
  NamedRerouteUsage()
  NamedRerouteUsage()
  NamedRerouteUsage()
  Normalize(VectorInput=@n51)
  SetMaterialAttributes!(MaterialAttributes=@n41, BaseColor=NamedRerouteUsage()) {attributeSetTypes [#4]}
  TransformPosition(None=WorldPosition() {worldPositionShaderOffset=WPT_CameraRelative}.XYZ) {transformSourceType=TRANSFORMPOSSOURCE_TranslatedWorld}
 ]}
