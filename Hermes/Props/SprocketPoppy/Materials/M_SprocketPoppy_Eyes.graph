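# Additive unlit material for the glowing eyes of the SprocketPoppy prop, drawn as two summed radial falloffs with a squint scale and depth fade.
# World position offset pivots each eye, split by local-position sign, toward the player position from a parameter collection.
# The face animation function supplies the look-at mask, extra transform and a random mask that collapses vertices to the local origin where zero.
# Destructible damage animation is added to the offset.
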Material /Hermes/Props/SprocketPoppy/Materials/M_SprocketPoppy_Eyes
graph {
 properties {bAutomaticallySetUsageInEditor=false, bUseMaterialAttributes=true, blendMode=BLEND_Additive, shadingModel=MSM_Unlit, usedShadingModels=Unlit}
 nodes [
  CameraPositionWS()
  CameraPositionWS()
  n1 = Subtract(
   A=CameraPositionWS()
   B=TransformPosition(None=VectorParameter() {parameterName=Pivot, defaultValue=<rgba 0 0 136.2579 1>}.RGB) {transformType=TRANSFORMPOSSOURCE_World}
  )
  ComponentMask(None=Transform(None=Normalize(VectorInput=@n1)) {transformSourceType=TRANSFORMSOURCE_World, transformType=TRANSFORM_Tangent}) {g=true, r=true}
  Constant3Vector() {constant=<rgba 0 1 0 0>}
  n2 = Time()
  Multiply(A=Sine(None=Multiply(A=@n2)), B=ScalarParameter() {defaultValue=0.08, parameterName=RotateTest})
  n3 = CollectionParameter() {collection=&FortniteMaterialParameters, parameterName=PlayerPosition}
  n4 = MaterialFunctionCall(Float3=LocalPosition() {includedOffsets=ExcludeOffsets}.XYZ) {materialFunction=&BreakOutFloat3Components}
  n5 = Sign(None=@n4.G)
  n6 = Constant()
  n7 = Multiply(A=ScalarParameter() {defaultValue=-20.0014, parameterName=EyeCenterBias}, B=OneMinus(None=Saturate(None=@n5)))
  n8 = Multiply(
   A=MaterialFunctionCall(X=Sign(None=@n4.R), Y=@n5, Z=Constant() {r=1}) {materialFunction=&MakeFloat3}.Result
   B=Add(
    A=MaterialFunctionCall(X=@n6, Y=@n7, Z=@n6) {materialFunction=&MakeFloat3}.Result
    B=VectorParameter() {parameterName=PivotEyeBall, defaultValue=<rgba 31.8848 74.6959 116.9895 1>}.RGB
   )
  )
  n9 = Subtract(
   A=@n8
   B=Add(
    A=VectorParameter() {parameterName=EyeBias, defaultValue=<rgba 0 0 -283.0042 1>}.RGB
    B=TransformPosition(None=@n3) {transformSourceType=TRANSFORMPOSSOURCE_World}
   )
  )
  n10 = Normalize(VectorInput=@n9)
  n11 = Constant3Vector() {constant=<rgba 0 1 0 0>}
  Multiply(A=@n10, B=Sign(None=DotProduct(A=@n11, B=@n10)))
  n12 = MaterialFunctionCall!(In) {materialFunction=/Hermes/Props/SprocketPoppy/Materials/FaceAnim/MF_SprocketPoppy_Anim}
  n13 = Subtract(A=LocalPosition() {includedOffsets=ExcludeOffsets}.XYZ, B=@n8)
  n14 = Normalize(VectorInput=Multiply(A=Multiply(A=@n5) {constB=-1}, B=@n9))
  n15 = Constant3Vector() {constant=<rgba 0 0 1 0>}
  n16 = MaterialFunctionCall(VectorToTransform=@n13, BasisX=Normalize(VectorInput=CrossProduct(A=@n14, B=@n15)), BasisY=@n14, BasisZ=@n15) {materialFunction=&Transform3x3Matrix}
  NamedRerouteDeclaration(None=Multiply(A=@n12.LookAtTransition mask, B=Subtract(A=@n16.Transformed Vector, B=@n13))) {name=EyeLocalTransform, nodeColor=<rgba 0.2656 1 0 1>}
  n17 = Constant3Vector() {constant=<rgba 1 0 0 0>}
  n18 = Power(Base=Abs(None=DotProduct(A=@n10, B=@n17)), Exp=ScalarParameter() {defaultValue=8.6339, parameterName=LookAtFade})
  NamedRerouteDeclaration(None=OneMinus(None=Saturate(None=@n18))) {name=Debug color, nodeColor=<rgba 0.6875 1 0 1>}
  n19 = CrossProduct(A=@n14, B=@n15)
  Normalize(VectorInput=CrossProduct(A=Normalize(VectorInput=@n19) {desc=X}, B=@n14)) {desc=Z}
  Normalize()
  RotateAboutAxis(
   NormalizedRotationAxis=Constant3Vector() {constant=<rgba 0 0 1 0>}
   PivotPoint=@n8
   Position=LocalPosition() {includedOffsets=ExcludeOffsets}.XYZ
  )
  n20 = VectorParameter() {parameterName=EyeGlow2, defaultValue=<rgba 16.9523 20 4.1886 1>}
  n21 = Constant() {r=1}
  n22 = ScalarParameter() {defaultValue=0.888, parameterName=EyeSquint}
  n23 = Distance(
   A=MaterialFunctionCall(UVs=TextureCoordinate(), Texture Scale=AppendVector(A=@n21, B=@n22)) {materialFunction=&ScaleUVsByCenter}.UVs
   B=Constant2Vector() {g=0.5, r=0.5}
  )
  n24 = Divide(A=@n23, B=ScalarParameter() {defaultValue=0.088, parameterName=EyeFalloffDist2})
  n25 = Power(Base=Saturate(None=OneMinus(None=@n24)), Exp=ScalarParameter() {defaultValue=2.8339, parameterName=EyeFalloffPow2})
  n26 = VectorParameter() {parameterName=EyeGlow, defaultValue=<rgba 0.27 0.4062 0.1699 1>}
  n27 = Divide(A=@n23, B=ScalarParameter() {defaultValue=0.4703, parameterName=EyeFalloffDist1})
  n28 = Power(Base=Saturate(None=OneMinus(None=@n27)), Exp=ScalarParameter() {defaultValue=2.4437, parameterName=EyeFalloffPow1})
  n29 = LinearInterpolate(A=Multiply(A=LocalPosition().XYZ) {constB=-1}, B=Add(A=@n12.LookatTransformLocalOnly, B=NamedRerouteUsage()), Alpha=@n12.RandomMask)
  SetMaterialAttributes!(
   MaterialAttributes
   Emissive Color=Add(A=Multiply(A=@n20.RGB, B=@n25), B=Multiply(A=@n26.RGB, B=@n28))
   Opacity=DepthFade(Opacity=Constant() {r=1}, FadeDistance=ScalarParameter() {defaultValue=17.5615, parameterName=DepthFade}) {fadeDistanceDefault=1}
   World Position Offset=Add(
    A=Transform(None=@n29) {transformSourceType=TRANSFORMSOURCE_Local}
    B=MaterialFunctionCall() {materialFunction=&MF_FN_DestructibleDamage}.DamageAnimWPO
   )
  ) {attributeSetTypes [#1, #2, #3]}
 ]
}
